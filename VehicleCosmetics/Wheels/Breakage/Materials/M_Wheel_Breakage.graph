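# M_Wheel_Breakage is the masked material for the Breakage wheel cosmetic, adding an animated blue emissive effect to the rim.
# It builds on MF_WheelParent_Simple, replacing emissive where the rim mask's alpha and inverted red vertex color overlap.
# The glow sums two counter-panning breakage masks and a panning water texture distorted by a scrolling normal map.
# The result then passes through the shared vehicle damage and de-rez transition functions.
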Material /VehicleCosmetics/Wheels/Breakage/Materials/M_Wheel_Breakage
graph {
 properties {
  bAutomaticallySetUsageInEditor=false
  bUseMaterialAttributes=true
  bUsedWithMeshDeformer=true
  bUsedWithSkeletalMesh=true
  blendMode=BLEND_Masked
  pixelDepthOffsetMode=PDOM_Legacy
  usedShadingModels=Default Lit
 }
 nodes [
  n1 = MaterialFunctionCall() {materialFunction=&MF_WheelParent_Simple}
  n2 = GetMaterialAttributes(None=@n1.MaterialAttributes) {attributeGetTypes [#1]}
  n3 = TextureCoordinate()
  n4 = TextureCoordinate()
  n5 = Add(
   A=TextureSample(UVs=Panner(Coordinate=@n3) {speedX=0.03, speedY=0.05}) {samplerType=SAMPLERTYPE_Masks, texture=/VehicleCosmetics/Wheels/Breakage/Textures/T_Rim_Breakage_Mask}.R
   B=TextureSample(UVs=Panner(Coordinate=@n4) {speedX=0.03, speedY=-0.05}) {samplerType=SAMPLERTYPE_Masks, texture=&T_Rim_Breakage_Mask}.R
  )
  n6 = Constant3Vector() {constant=<rgba 0 0.41 1 0>}
  n7 = TextureSample(UVs=Panner(Coordinate=TextureCoordinate() {uTiling=20, vTiling=20}) {speedX=-0.16, speedY=-0.1}) {samplerType=SAMPLERTYPE_Normal, texture=/VehicleCosmetics/Wheels/_SharedTextures/FX/Water_02_N}
  n8 = TextureCoordinate() {uTiling=0.05, vTiling=0.05}
  n9 = TextureSample(UVs=Panner(Coordinate=Add(A=@n7.R, B=@n8)) {speedX=0.05, speedY=0.05}) {texture=/VehicleCosmetics/Wheels/_SharedTextures/FX/Water_01_Pack}
  n10 = Constant3Vector() {constant=<rgba 0.14 0.34 0.47 0>}
  n11 = LinearInterpolate(
   A=@n2.Emissive Color
   B=Add(A=Multiply(A=@n6, B=@n5), B=Multiply(A=@n9.R, B=@n10))
   Alpha=Multiply(A=ComponentMask(None=@n1.RimRGBAMask) {a=true}, B=OneMinus(None=@n1.VertexColorR))
  )
  MaterialFunctionCall(In=MaterialFunctionCall(Material=SetMaterialAttributes(MaterialAttributes=@n2.MaterialAttributes, Emissive Color=@n11) {attributeSetTypes [#1]}) {materialFunction=&MF_BaseVehicle_Damage}.Result) {materialFunction=&MF_VehicleDeRez_Transition}
 ]
}
